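# Material function for the SonnetSpirit backpack that draws animated anime eyes over incoming Lit and Shaded colors.
# It derives eye masks from UV1, then offsets and scales per-eye coordinates using look-pose parameters to place a procedural SphereMask pupil.
# A rounded eye_pose selects channels of a flipbook mask texture for eyelid and outline shapes.
# UseEyeAnimation and pupil distance-scaling switches exist but feed nothing.
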MaterialFunction /BRCosmetics/Accessories/FORT_Backpacks/Backpack_M_LRG_SonnetSpirit/Materials/MF_SonnetSpirit_Backpack_AnimeEyeAnimation
graph {expressions {
  MaterialExpressionAdd_11 {
   class=MaterialExpressionAdd
   inputs [
    A {from=MaterialExpressionScalarParameter_20}
    B {from=MaterialExpressionMultiply_13}
   ]
  }
  MaterialExpressionAdd_13 {
   class=MaterialExpressionAdd
   inputs [
    A {from=MaterialExpressionNamedRerouteUsage_43}
    B {from=""}
   ]
   properties {constB=-0.5}
  }
  MaterialExpressionAdd_15 {
   class=MaterialExpressionAdd
   inputs [
    A {from=MaterialExpressionMaterialFunctionCall_2, output=R}
    B {from=""}
   ]
   properties {constB=-1}
  }
  MaterialExpressionAdd_16 {
   class=MaterialExpressionAdd
   inputs [
    A {from=MaterialExpressionMaterialFunctionCall_16, output=G}
    B {from=""}
   ]
   properties {constB=-0.5}
  }
  MaterialExpressionAdd_17 {
   class=MaterialExpressionAdd
   inputs [
    A {from=MaterialExpressionMultiply_5}
    B {from=""}
   ]
   properties {constB=0.5}
  }
  MaterialExpressionAdd_20 {
   class=MaterialExpressionAdd
   inputs [
    A {from=MaterialExpressionNamedRerouteUsage_24}
    B {from=""}
   ]
   properties {constB=0}
  }
  MaterialExpressionAdd_22 {
   class=MaterialExpressionAdd
   inputs [
    A {from=MaterialExpressionMultiply_35}
    B {from=""}
   ]
   properties {constB=0.5}
  }
  MaterialExpressionAdd_4 {
   class=MaterialExpressionAdd
   inputs [
    A {from=MaterialExpressionMaterialFunctionCall_17, output=Result}
    B {from=MaterialExpressionMultiply_27}
   ]
   properties {desc=move right eye to UV space 0-1}
  }
  MaterialExpressionAdd_5 {
   class=MaterialExpressionAdd
   inputs [
    A {from=MaterialExpressionAdd_4}
    B {from=MaterialExpressionAppendVector_0}
   ]
  }
  MaterialExpressionAdd_9 {
   class=MaterialExpressionAdd
   inputs [
    A {from=MaterialExpressionScalarParameter_26}
    B {from=MaterialExpressionMultiply_16}
   ]
  }
  MaterialExpressionAppendVector_0 {
   class=MaterialExpressionAppendVector
   inputs [
    A {from=MaterialExpressionMultiply_40, output=""}
    B {from=MaterialExpressionMultiply_41, output=""}
   ]
  }
  MaterialExpressionCeil_1 {class=MaterialExpressionCeil, inputs [None {from=MaterialExpressionMaterialFunctionCall_4, output=G}], properties {}}
  MaterialExpressionClamp_2 {
   class=MaterialExpressionClamp
   inputs [
    None {from=MaterialExpressionAdd_9, output=""}
    Min {from="", output=""}
    Max {from="", output=""}
   ]
   properties {minDefault=-1}
  }
  MaterialExpressionClamp_3 {
   class=MaterialExpressionClamp
   inputs [
    None {from=MaterialExpressionAdd_11, output=""}
    Min {from="", output=""}
    Max {from="", output=""}
   ]
   properties {minDefault=-1}
  }
  MaterialExpressionConstant2Vector_0 {class=MaterialExpressionConstant2Vector, properties {r=-1}}
  MaterialExpressionConstant_0 {class=MaterialExpressionConstant}
  MaterialExpressionConstant_1 {class=MaterialExpressionConstant}
  MaterialExpressionConstant_2 {class=MaterialExpressionConstant, properties {r=1}}
  MaterialExpressionConvert_0 {
   class=MaterialExpressionConvert
   inputs [None {from=MaterialExpressionFrac_0}]
   properties {
    nodeName=Break Float2
    convertInputs [{
      type=Vector2
      expressionInput {
       expression=&self:MaterialExpressionFrac_0
       inputName=None
       mask=0
       maskA=0
       maskB=0
       maskG=0
       maskR=0
       outputIndex=0
      }
     }]
    convertOutputs [{type=Scalar}, {type=Scalar}]
   }
  }
  MaterialExpressionConvert_1 {
   class=MaterialExpressionConvert
   inputs [
    None {from=MaterialExpressionOneMinus_0}
    None {from=MaterialExpressionConvert_0}
   ]
   properties {
    nodeName=Make Vector2
    convertInputs [
     {
      type=Scalar
      expressionInput {
       expression=&self:MaterialExpressionOneMinus_0
       inputName=None
       mask=0
       maskA=0
       maskB=0
       maskG=0
       maskR=0
       outputIndex=0
      }
     }
     {
      type=Scalar
      expressionInput {
       expression=&self:MaterialExpressionConvert_0
       inputName=None
       mask=0
       maskA=0
       maskB=0
       maskG=0
       maskR=0
       outputIndex=1
      }
     }
    ]
    convertOutputs [{type=Vector2}]
   }
  }
  MaterialExpressionDistance_0 {
   class=MaterialExpressionDistance
   inputs [
    A {from=MaterialExpressionMultiply_4, output=""}
    B {from=MaterialExpressionConstant_0, output=""}
   ]
   properties {}
  }
  MaterialExpressionFloor_0 {class=MaterialExpressionFloor, inputs [None {from=MaterialExpressionAdd_20, output=""}], properties {}}
  MaterialExpressionFloor_1 {class=MaterialExpressionFloor, inputs [None {from=MaterialExpressionMaterialFunctionCall_4, output=R}], properties {desc=LeftEye}}
  MaterialExpressionFrac_0 {class=MaterialExpressionFrac, inputs [None {from=MaterialExpressionNamedRerouteUsage_1, output=""}], properties {}}
  MaterialExpressionFunctionInput_0 {
   class=MaterialExpressionFunctionInput
   inputs [Preview {from="", output=""}]
   properties {bUsePreviewValueAsDefault=true, inputName=Mask, inputType=FunctionInput_Scalar, sortPriority=20, previewValue=<1 0 0 1>}
  }
  MaterialExpressionFunctionInput_2 {class=MaterialExpressionFunctionInput, inputs [Preview {from="", output=""}], properties {inputName=Color Lit, sortPriority=1}}
  MaterialExpressionFunctionInput_3 {class=MaterialExpressionFunctionInput, inputs [Preview {from="", output=""}], properties {inputName=Color Shaded, sortPriority=2}}
  MaterialExpressionFunctionOutput_1 {
   class=MaterialExpressionFunctionOutput
   inputs [None {from=MaterialExpressionLinearInterpolate_16, output=""}]
   properties {outputName=Lit, sortPriority=1}
  }
  MaterialExpressionFunctionOutput_2 {
   class=MaterialExpressionFunctionOutput
   inputs [None {from=MaterialExpressionLinearInterpolate_15, output=""}]
   properties {outputName=Shaded, sortPriority=2}
  }
  MaterialExpressionIf_0 {
   class=MaterialExpressionIf
   inputs [
    A {from=MaterialExpressionRound_0, output=""}
    B {from="", output=""}
    "A > B" {from=MaterialExpressionTextureSample_0, output=B}
    "A == B" {from="", output=""}
    "A < B" {from=MaterialExpressionRound_0, output=""}
   ]
   properties {constB=2.5, desc=eyelid down when sad}
  }
  MaterialExpressionIf_1 {
   class=MaterialExpressionIf
   inputs [
    A {from=MaterialExpressionRound_0, output=""}
    B {from="", output=""}
    "A > B" {from=MaterialExpressionTextureSample_0, output=A}
    "A == B" {from="", output=""}
    "A < B" {from=MaterialExpressionLinearInterpolate_23, output=""}
   ]
   properties {constB=2.5}
  }
  MaterialExpressionIf_2 {
   class=MaterialExpressionIf
   inputs [
    A {from=MaterialExpressionRound_0, output=""}
    B {from="", output=""}
    "A > B" {from=MaterialExpressionTextureSample_0, output=G}
    "A == B" {from="", output=""}
    "A < B" {from=MaterialExpressionTextureSample_0, output=G}
   ]
   properties {constB=1.5}
  }
  MaterialExpressionLinearInterpolate_0 {
   class=MaterialExpressionLinearInterpolate
   inputs [
    A {from=MaterialExpressionScalarParameter_0}
    B {from=MaterialExpressionScalarParameter_30}
    Alpha {from=MaterialExpressionMaterialFunctionCall_1, output=Result}
   ]
  }
  MaterialExpressionLinearInterpolate_12 {
   class=MaterialExpressionLinearInterpolate
   inputs [
    A {from=MaterialExpressionVectorParameter_1, output=RGB}
    B {from=MaterialExpressionLinearInterpolate_6}
    Alpha {from=MaterialExpressionSubtract_2}
   ]
  }
  MaterialExpressionLinearInterpolate_13 {
   class=MaterialExpressionLinearInterpolate
   inputs [
    A {from=MaterialExpressionReroute_0}
    B {from=MaterialExpressionMaterialFunctionCall_12, output=Result}
    Alpha {from=MaterialExpressionNamedRerouteUsage_23}
   ]
   properties {desc=rotate right eye UVs}
  }
  MaterialExpressionLinearInterpolate_15 {
   class=MaterialExpressionLinearInterpolate
   inputs [
    A {from=MaterialExpressionNamedRerouteDeclaration_26}
    B {from=MaterialExpressionNamedRerouteUsage_34}
    Alpha {from=MaterialExpressionNamedRerouteUsage_33}
   ]
  }
  MaterialExpressionLinearInterpolate_16 {
   class=MaterialExpressionLinearInterpolate
   inputs [
    A {from=MaterialExpressionNamedRerouteDeclaration_25}
    B {from=MaterialExpressionNamedRerouteUsage_41}
    Alpha {from=MaterialExpressionNamedRerouteUsage_35}
   ]
  }
  MaterialExpressionLinearInterpolate_19 {
   class=MaterialExpressionLinearInterpolate
   inputs [
    A {from=MaterialExpressionNamedRerouteUsage_1}
    B {from=MaterialExpressionConvert_1}
    Alpha {from=MaterialExpressionNamedRerouteUsage_32}
   ]
  }
  MaterialExpressionLinearInterpolate_21 {
   class=MaterialExpressionLinearInterpolate
   inputs [
    A {from=MaterialExpressionMultiply_12}
    B {from=MaterialExpressionVectorParameter_5, output=RGB}
    Alpha {from=MaterialExpressionNamedRerouteUsage_40}
   ]
  }
  MaterialExpressionLinearInterpolate_22 {
   class=MaterialExpressionLinearInterpolate
   inputs [
    A {from=MaterialExpressionLinearInterpolate_7}
    B {from=MaterialExpressionVectorParameter_5, output=RGB}
    Alpha {from=MaterialExpressionNamedRerouteUsage_40}
   ]
  }
  MaterialExpressionLinearInterpolate_23 {
   class=MaterialExpressionLinearInterpolate
   inputs [
    A {from=""}
    B {from=MaterialExpressionIf_2}
    Alpha {from=MaterialExpressionRound_0}
   ]
  }
  MaterialExpressionLinearInterpolate_6 {
   class=MaterialExpressionLinearInterpolate
   inputs [
    A {from=MaterialExpressionVectorParameter_4, output=RGB}
    B {from=MaterialExpressionVectorParameter_3, output=RGB}
    Alpha {from=MaterialExpressionReroute_4}
   ]
  }
  MaterialExpressionLinearInterpolate_7 {
   class=MaterialExpressionLinearInterpolate
   inputs [
    A {from=MaterialExpressionVectorParameter_1, output=RGB}
    B {from=MaterialExpressionLinearInterpolate_6}
    Alpha {from=MaterialExpressionSubtract_2}
   ]
  }
  MaterialExpressionMaterialFunctionCall_0 {
   class=MaterialExpressionMaterialFunctionCall
   inputs [
    UVs {from=MaterialExpressionAdd_5}
    Texture Scale {from=MaterialExpressionScalarParameter_0}
   ]
   properties {materialFunction=&ScaleUVsByCenter}
  }
  MaterialExpressionMaterialFunctionCall_1 {
   class=MaterialExpressionMaterialFunctionCall
   inputs [
    Fade Length {from=MaterialExpressionScalarParameter_32}
    Fade Offset {from=MaterialExpressionScalarParameter_31}
    For Vertex Shader {from=""}
   ]
   properties {materialFunction=&CameraDepthFade}
  }
  MaterialExpressionMaterialFunctionCall_10 {
   class=MaterialExpressionMaterialFunctionCall
   inputs [Float2 {from=MaterialExpressionMaterialFunctionCall_0, output=UVs}]
   properties {materialFunction=&BreakOutFloat2Components}
  }
  MaterialExpressionMaterialFunctionCall_11 {
   class=MaterialExpressionMaterialFunctionCall
   inputs [
    X {from=MaterialExpressionAdd_22}
    Y {from=MaterialExpressionMaterialFunctionCall_10, output=G}
   ]
   properties {materialFunction=&MakeFloat2}
  }
  MaterialExpressionMaterialFunctionCall_12 {
   class=MaterialExpressionMaterialFunctionCall
   inputs [
    X {from=MaterialExpressionMaterialFunctionCall_9, output=G}
    Y {from=MaterialExpressionMaterialFunctionCall_9, output=G}
   ]
   properties {materialFunction=&MakeFloat2}
  }
  MaterialExpressionMaterialFunctionCall_16 {
   class=MaterialExpressionMaterialFunctionCall
   inputs [Float2 {from=MaterialExpressionNamedRerouteUsage_2}]
   properties {materialFunction=&BreakOutFloat2Components}
  }
  MaterialExpressionMaterialFunctionCall_17 {
   class=MaterialExpressionMaterialFunctionCall
   inputs [
    X {from=MaterialExpressionMaterialFunctionCall_16, output=R}
    Y {from=MaterialExpressionAdd_17}
   ]
   properties {materialFunction=&MakeFloat2}
  }
  MaterialExpressionMaterialFunctionCall_2 {
   class=MaterialExpressionMaterialFunctionCall
   inputs [Float2 {from=MaterialExpressionFloor_0}]
   properties {materialFunction=&BreakOutFloat2Components}
  }
  MaterialExpressionMaterialFunctionCall_4 {
   class=MaterialExpressionMaterialFunctionCall
   inputs [Float2 {from=MaterialExpressionNamedRerouteUsage_22}]
   properties {materialFunction=&BreakOutFloat2Components}
  }
  MaterialExpressionMaterialFunctionCall_7 {
   class=MaterialExpressionMaterialFunctionCall
   inputs [Float2 {from=MaterialExpressionMultiply_3}]
   properties {materialFunction=&BreakOutFloat2Components}
  }
  MaterialExpressionMaterialFunctionCall_9 {
   class=MaterialExpressionMaterialFunctionCall
   inputs [Float2 {from=MaterialExpressionReroute_0}]
   properties {materialFunction=&BreakOutFloat2Components}
  }
  MaterialExpressionMax_11 {
   class=MaterialExpressionMax
   inputs [
    A {from=MaterialExpressionAdd_15}
    B {from=MaterialExpressionMaterialFunctionCall_2, output=G}
   ]
   properties {constB=0}
  }
  MaterialExpressionMax_12 {
   class=MaterialExpressionMax
   inputs [
    A {from=MaterialExpressionMaterialFunctionCall_7, output=R}
    B {from=MaterialExpressionMaterialFunctionCall_7, output=R}
   ]
   properties {constB=0}
  }
  MaterialExpressionMax_13 {
   class=MaterialExpressionMax
   inputs [
    A {from=MaterialExpressionMax_11}
    B {from=MaterialExpressionMax_12}
   ]
   properties {constB=0}
  }
  MaterialExpressionMin_0 {
   class=MaterialExpressionMin
   inputs [
    A {from=MaterialExpressionCeil_1, output=""}
    B {from=MaterialExpressionNamedRerouteUsage_5, output=""}
   ]
   properties {}
  }
  MaterialExpressionMultiply_12 {
   class=MaterialExpressionMultiply
   inputs [
    A {from=MaterialExpressionLinearInterpolate_12}
    B {from=MaterialExpressionScalarParameter_36}
   ]
  }
  MaterialExpressionMultiply_13 {
   class=MaterialExpressionMultiply
   inputs [
    A {from=MaterialExpressionScalarParameter_27}
    B {from=""}
   ]
   properties {constB=-1}
  }
  MaterialExpressionMultiply_16 {
   class=MaterialExpressionMultiply
   inputs [
    A {from=MaterialExpressionScalarParameter_23}
    B {from=""}
   ]
   properties {constB=-1}
  }
  MaterialExpressionMultiply_20 {
   class=MaterialExpressionMultiply
   inputs [
    A {from=MaterialExpressionMin_0}
    B {from=MaterialExpressionFunctionInput_0}
   ]
  }
  MaterialExpressionMultiply_27 {
   class=MaterialExpressionMultiply
   inputs [
    A {from=MaterialExpressionNamedRerouteUsage_3}
    B {from=MaterialExpressionConstant2Vector_0}
   ]
  }
  MaterialExpressionMultiply_3 {
   class=MaterialExpressionMultiply
   inputs [
    A {from=MaterialExpressionFloor_0}
    B {from=""}
   ]
   properties {constB=-1}
  }
  MaterialExpressionMultiply_35 {
   class=MaterialExpressionMultiply
   inputs [
    A {from=MaterialExpressionSubtract_0}
    B {from=""}
   ]
   properties {constB=2}
  }
  MaterialExpressionMultiply_4 {
   class=MaterialExpressionMultiply
   inputs [
    A {from=MaterialExpressionAdd_13}
    B {from=""}
   ]
   properties {constB=2}
  }
  MaterialExpressionMultiply_40 {
   class=MaterialExpressionMultiply
   inputs [
    A {from=MaterialExpressionClamp_2}
    B {from=MaterialExpressionScalarParameter_21}
   ]
  }
  MaterialExpressionMultiply_41 {
   class=MaterialExpressionMultiply
   inputs [
    A {from=MaterialExpressionClamp_3}
    B {from=MaterialExpressionScalarParameter_18}
   ]
  }
  MaterialExpressionMultiply_5 {
   class=MaterialExpressionMultiply
   inputs [
    A {from=MaterialExpressionAdd_16}
    B {from=MaterialExpressionScalarParameter_7}
   ]
  }
  MaterialExpressionMultiply_6 {
   class=MaterialExpressionMultiply
   inputs [
    A {from=MaterialExpressionOneMinus_2}
    B {from=MaterialExpressionNamedRerouteDeclaration_2}
   ]
   properties {desc=RightEye}
  }
  MaterialExpressionNamedRerouteDeclaration_0 {
   class=MaterialExpressionNamedRerouteDeclaration
   inputs [None {from=MaterialExpressionTextureCoordinate_5}]
   properties {name=EyeUVs, nodeColor=<rgba 0.2188 0 1 1>}
  }
  MaterialExpressionNamedRerouteDeclaration_1 {
   class=MaterialExpressionNamedRerouteDeclaration
   inputs [None {from=MaterialExpressionSaturate_1}]
   properties {name=DF_Eyelid, nodeColor=<rgba 0 1 0.3672 1>}
  }
  MaterialExpressionNamedRerouteDeclaration_11 {
   class=MaterialExpressionNamedRerouteDeclaration
   inputs [None {from=MaterialExpressionLinearInterpolate_22}]
   properties {name=EyeColor_Lit, nodeColor=<rgba 1 0.5322 0.1287 1>}
  }
  MaterialExpressionNamedRerouteDeclaration_12 {
   class=MaterialExpressionNamedRerouteDeclaration
   inputs [None {from=MaterialExpressionOneMinus_1}]
   properties {name=Mask_EyesUDIM, nodeColor=<rgba 1 0.2578 0 1>}
  }
  MaterialExpressionNamedRerouteDeclaration_15 {
   class=MaterialExpressionNamedRerouteDeclaration
   inputs [None {from=MaterialExpressionSaturate_0}]
   properties {name=Mask_Eyelids, nodeColor=<rgba 0 1 1 1>}
  }
  MaterialExpressionNamedRerouteDeclaration_2 {
   class=MaterialExpressionNamedRerouteDeclaration
   inputs [None {from=MaterialExpressionMultiply_20}]
   properties {name=Mask_Eyes, nodeColor=<rgba 1 0 0.8906 1>}
  }
  MaterialExpressionNamedRerouteDeclaration_24 {
   class=MaterialExpressionNamedRerouteDeclaration
   inputs [None {from=MaterialExpressionLinearInterpolate_13}]
   properties {name=EyeCoords, nodeColor=<rgba 0 1 0.6016 1>}
  }
  MaterialExpressionNamedRerouteDeclaration_25 {
   class=MaterialExpressionNamedRerouteDeclaration
   inputs [None {from=MaterialExpressionFunctionInput_2}]
   properties {name=InColorLit, nodeColor=<rgba 1 0 0.0234 1>}
  }
  MaterialExpressionNamedRerouteDeclaration_26 {
   class=MaterialExpressionNamedRerouteDeclaration
   inputs [None {from=MaterialExpressionFunctionInput_3}]
   properties {name=InColorShaded, nodeColor=<rgba 0 0.3437 1 1>}
  }
  MaterialExpressionNamedRerouteDeclaration_27 {
   class=MaterialExpressionNamedRerouteDeclaration
   inputs [None {from=MaterialExpressionLinearInterpolate_21}]
   properties {name=EyeColor_Shaded, nodeColor=<rgba 1 0.5322 0.1287 1>}
  }
  MaterialExpressionNamedRerouteDeclaration_3 {
   class=MaterialExpressionNamedRerouteDeclaration
   inputs [None {from=MaterialExpressionMultiply_6}]
   properties {name=Mask_EyeRight, nodeColor=<rgba 1 0 0.0152 1>}
  }
  MaterialExpressionNamedRerouteDeclaration_4 {
   class=MaterialExpressionNamedRerouteDeclaration
   inputs [None {from=MaterialExpressionFloor_1}]
   properties {name=Mask_EyeLeft, nodeColor=<rgba 0.0293 0.3125 0.0353 1>}
  }
  MaterialExpressionNamedRerouteUsage_1 {class=MaterialExpressionNamedRerouteUsage}
  MaterialExpressionNamedRerouteUsage_2 {class=MaterialExpressionNamedRerouteUsage}
  MaterialExpressionNamedRerouteUsage_22 {class=MaterialExpressionNamedRerouteUsage}
  MaterialExpressionNamedRerouteUsage_23 {class=MaterialExpressionNamedRerouteUsage}
  MaterialExpressionNamedRerouteUsage_24 {class=MaterialExpressionNamedRerouteUsage}
  MaterialExpressionNamedRerouteUsage_3 {class=MaterialExpressionNamedRerouteUsage}
  MaterialExpressionNamedRerouteUsage_32 {class=MaterialExpressionNamedRerouteUsage}
  MaterialExpressionNamedRerouteUsage_33 {class=MaterialExpressionNamedRerouteUsage}
  MaterialExpressionNamedRerouteUsage_34 {class=MaterialExpressionNamedRerouteUsage}
  MaterialExpressionNamedRerouteUsage_35 {class=MaterialExpressionNamedRerouteUsage}
  MaterialExpressionNamedRerouteUsage_40 {class=MaterialExpressionNamedRerouteUsage}
  MaterialExpressionNamedRerouteUsage_41 {class=MaterialExpressionNamedRerouteUsage}
  MaterialExpressionNamedRerouteUsage_43 {class=MaterialExpressionNamedRerouteUsage}
  MaterialExpressionNamedRerouteUsage_44 {class=MaterialExpressionNamedRerouteUsage}
  MaterialExpressionNamedRerouteUsage_46 {class=MaterialExpressionNamedRerouteUsage}
  MaterialExpressionNamedRerouteUsage_5 {class=MaterialExpressionNamedRerouteUsage}
  MaterialExpressionOneMinus_0 {class=MaterialExpressionOneMinus, inputs [None {from=MaterialExpressionConvert_0}]}
  MaterialExpressionOneMinus_1 {class=MaterialExpressionOneMinus, inputs [None {from=MaterialExpressionMax_13}]}
  MaterialExpressionOneMinus_2 {class=MaterialExpressionOneMinus, inputs [None {from=MaterialExpressionFloor_1}]}
  MaterialExpressionOneMinus_6 {class=MaterialExpressionOneMinus, inputs [None {from=MaterialExpressionScalarParameter_22}]}
  MaterialExpressionReroute_0 {class=MaterialExpressionReroute, inputs [None {from=MaterialExpressionMaterialFunctionCall_11, output=Result}]}
  MaterialExpressionReroute_4 {class=MaterialExpressionReroute, inputs [None {from=MaterialExpressionSaturate_14}]}
  MaterialExpressionRound_0 {class=MaterialExpressionRound, inputs [None {from=MaterialExpressionScalarParameter_34, output=""}], properties {}}
  MaterialExpressionSaturate_0 {class=MaterialExpressionSaturate, inputs [None {from=MaterialExpressionIf_0}]}
  MaterialExpressionSaturate_1 {class=MaterialExpressionSaturate, inputs [None {from=MaterialExpressionIf_1}]}
  MaterialExpressionSaturate_13 {class=MaterialExpressionSaturate, inputs [None {from=MaterialExpressionDistance_0}]}
  MaterialExpressionSaturate_14 {class=MaterialExpressionSaturate, inputs [None {from=MaterialExpressionSphereMask_0}]}
  MaterialExpressionScalarParameter_0 {
   class=MaterialExpressionScalarParameter
   properties {defaultValue=0.5, group=01_EyeAnimation_Parameters, parameterName=PupilMap_Scale, sliderMax=2, sortPriority=5}
  }
  MaterialExpressionScalarParameter_18 {
   class=MaterialExpressionScalarParameter
   properties {defaultValue=0.2, group=01_EyeAnimation_Parameters, parameterName=EyeVertical_Threshold, sliderMax=1}
  }
  MaterialExpressionScalarParameter_20 {class=MaterialExpressionScalarParameter, properties {group=01_EyeAnimation_Poses, parameterName=look_up_pose, sliderMax=1, sortPriority=12}}
  MaterialExpressionScalarParameter_21 {
   class=MaterialExpressionScalarParameter
   properties {defaultValue=0.2, group=01_EyeAnimation_Parameters, parameterName=EyeHorizontal_Threshold, sliderMax=1}
  }
  MaterialExpressionScalarParameter_22 {
   class=MaterialExpressionScalarParameter
   properties {defaultValue=0.4, group=01_EyeAnimation_Parameters, parameterName=Pupil_DF_LineThickness, sliderMax=1}
  }
  MaterialExpressionScalarParameter_23 {
   class=MaterialExpressionScalarParameter
   properties {group=01_EyeAnimation_Poses, parameterName=look_right_pose, sliderMax=1, sortPriority=10}
  }
  MaterialExpressionScalarParameter_26 {
   class=MaterialExpressionScalarParameter
   properties {group=01_EyeAnimation_Poses, parameterName=look_left_pose, sliderMax=1, sortPriority=11}
  }
  MaterialExpressionScalarParameter_27 {
   class=MaterialExpressionScalarParameter
   properties {group=01_EyeAnimation_Poses, parameterName=look_down_pose, sliderMax=1, sortPriority=13}
  }
  MaterialExpressionScalarParameter_30 {
   class=MaterialExpressionScalarParameter
   properties {defaultValue=0.5, group=01_EyeAnimation_Parameters, parameterName=PupilMap_ScaleFar, sliderMax=2, sortPriority=5}
  }
  MaterialExpressionScalarParameter_31 {
   class=MaterialExpressionScalarParameter
   properties {defaultValue=500, group=01_EyeAnimation_Parameters, parameterName=PupilDistanceScaling_StartDistance, sliderMax=2, sortPriority=5}
  }
  MaterialExpressionScalarParameter_32 {
   class=MaterialExpressionScalarParameter
   properties {defaultValue=500, group=01_EyeAnimation_Parameters, parameterName=PupilDistanceScaling_FadeLength, sliderMax=2, sortPriority=5}
  }
  MaterialExpressionScalarParameter_34 {class=MaterialExpressionScalarParameter, properties {group=01_EyeAnimation_Poses, parameterName=eye_pose, sliderMax=3, sortPriority=2}}
  MaterialExpressionScalarParameter_36 {
   class=MaterialExpressionScalarParameter
   properties {defaultValue=0.5, group=01_Anime Maps, parameterName=Dynamic_ColorShaded_Darkening, sliderMax=1, sortPriority=11}
  }
  MaterialExpressionScalarParameter_42 {
   class=MaterialExpressionScalarParameter
   properties {defaultValue=0.9, group=01_EyeAnimation_Parameters, parameterName=Pupil_DF_LineHardness, sliderMax=1}
  }
  MaterialExpressionScalarParameter_7 {
   class=MaterialExpressionScalarParameter
   properties {defaultValue=2, desc=horizontal UV ratio, group=01_EyeAnimation_Parameters, parameterName=PupilMap_UV_Ratio, sliderMax=2, sortPriority=5}
  }
  MaterialExpressionSphereMask_0 {
   class=MaterialExpressionSphereMask
   inputs [
    A {from=MaterialExpressionSaturate_13, output=""}
    B {from=MaterialExpressionConstant_2, output=""}
    Radius {from=MaterialExpressionOneMinus_6, output=""}
    Hardness {from=MaterialExpressionScalarParameter_42, output=""}
   ]
   properties {attenuationRadius=0.1}
  }
  MaterialExpressionStaticSwitchParameter_0 {
   class=MaterialExpressionStaticSwitchParameter
   inputs [
    True {from=MaterialExpressionLinearInterpolate_0, output=""}
    False {from=MaterialExpressionScalarParameter_0, output=""}
   ]
   properties {group=01_EyeAnimation_Parameters, parameterName=UsePupilDistanceScaling, sortPriority=4}
  }
  MaterialExpressionStaticSwitchParameter_4 {
   class=MaterialExpressionStaticSwitchParameter
   inputs [
    True {from=MaterialExpressionLinearInterpolate_15, output=""}
    False {from=MaterialExpressionNamedRerouteDeclaration_26, output=""}
   ]
   properties {group=01_EyeAnimation_Parameters, parameterName=UseEyeAnimation, sortPriority=-1}
  }
  MaterialExpressionStaticSwitchParameter_5 {
   class=MaterialExpressionStaticSwitchParameter
   inputs [
    True {from=MaterialExpressionLinearInterpolate_16, output=""}
    False {from=MaterialExpressionNamedRerouteDeclaration_25, output=""}
   ]
   properties {group=01_EyeAnimation_Parameters, parameterName=UseEyeAnimation, sortPriority=-1}
  }
  MaterialExpressionSubtract_0 {
   class=MaterialExpressionSubtract
   inputs [
    A {from=MaterialExpressionMaterialFunctionCall_10, output=R}
    B {from=""}
   ]
   properties {constB=0.5}
  }
  MaterialExpressionSubtract_2 {
   class=MaterialExpressionSubtract
   inputs [
    A {from=MaterialExpressionNamedRerouteUsage_44}
    B {from=MaterialExpressionNamedRerouteUsage_46}
   ]
  }
  MaterialExpressionTextureCoordinate_5 {class=MaterialExpressionTextureCoordinate, properties {coordinateIndex=1}}
  MaterialExpressionTextureSample_0 {
   class=MaterialExpressionTextureSample
   inputs [
    UVs {from=MaterialExpressionLinearInterpolate_19}
    Tex {from=""}
    Apply View MipBias {from=""}
   ]
   properties {
    samplerType=SAMPLERTYPE_Masks
    texture=/BRCosmetics/Accessories/FORT_Backpacks/Backpack_M_LRG_SonnetSpirit/Textures/T_SonnetSpirit_Backpack_FlipbookEyes
   }
  }
  MaterialExpressionVectorParameter_0 {
   class=MaterialExpressionVectorParameter
   properties {group=01_EyeAnimation_Parameters, parameterName=EyeLid_Color_Shaded, sortPriority=3, defaultValue=<rgba 0.5972 0.3813 0.0343 1>}
  }
  MaterialExpressionVectorParameter_1 {
   class=MaterialExpressionVectorParameter
   properties {group=01_EyeAnimation_Parameters, parameterName=EyeLid_Color_Lit, sortPriority=3, defaultValue=<rgba 1 0.6445 0.0452 1>}
  }
  MaterialExpressionVectorParameter_3 {
   class=MaterialExpressionVectorParameter
   properties {group=01_EyeAnimation_Parameters, parameterName=Eyeball_Color, sortPriority=2, defaultValue=<rgba 1 1 1 0>}
  }
  MaterialExpressionVectorParameter_4 {class=MaterialExpressionVectorParameter, properties {group=01_EyeAnimation_Parameters, parameterName=Pupil_Color, sortPriority=2}}
  MaterialExpressionVectorParameter_5 {
   class=MaterialExpressionVectorParameter
   properties {group=01_EyeAnimation_Parameters, parameterName=EyeLid_Color_DF_Line, sortPriority=3, defaultValue=<rgba 0 0 0 1>}
  }
 }}
