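# Control Rig for the SonnetSpirit backpack's "pen" character that turns an eye flipbook bone into Eye Open, Eye Close and Eye Sad curves.
# A three-way sequence, built from two chained sequence nodes, reads the Get Delta Ref X location of pen_eye_flipbook in each branch.
# Clamped, inverted remaps over successive unit-spaced bands drive the curves, so each stays at one below its band and falls to zero across it.
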ControlRigBlueprint /BRCosmetics/Accessories/FORT_Backpacks/Backpack_M_LRG_SonnetSpirit/Meshes/M_LRG_SonnetSpirit_Pack_CtrlRig
graphs [
 RigVMFunctionLibrary {links [], nodes []}
 RigVMFunction_Sequence_1_ContainedGraph {nodes [
   n1 = RigVMFunctionEntryNode()
   RigVMFunction_Sequence = &RigVMFunction_Sequence(ExecuteContext=@n1.ExecuteContext)
   n2 = &RigVMFunction_Sequence(ExecuteContext=@RigVMFunction_Sequence.B)
   RigVMFunctionReturnNode(A=@RigVMFunction_Sequence.A, B=@n2.A, C=@n2.B)
  ]}
 RigVMModel {nodes [
   RigUnit_BeginExecution = &RigUnit_BeginExecution()
   RigVMFunction_Sequence = RigVMAggregateNode(ExecuteContext=@RigUnit_BeginExecution.ExecutePin)
   Get Delta Ref = RigVMFunctionReferenceNode(Item=(Type=Bone,Name="pen_eye_flipbook"), ExecuteContext=@RigVMFunction_Sequence.A)
   > /Script/ControlRig.RigUnit_SetCurveValue(
    Curve=Eye Open
    Value=&RigVMFunction_MathFloatRemap(SourceMaximum=0.9, SourceMinimum=0, TargetMaximum=0, TargetMinimum=1, bClamp=true, Value=@"Get Delta Ref".Result.Location.X).Result
   ) {is=struct RigVMUnitNode}
   Get Delta Ref_1 = RigVMFunctionReferenceNode(Item=(Type=Bone,Name="pen_eye_flipbook"), ExecuteContext=@RigVMFunction_Sequence.B)
   > &RigUnit_SetCurveValue(
    Curve=Eye Close
    Value=&RigVMFunction_MathFloatRemap(SourceMaximum=1.9, SourceMinimum=1, TargetMaximum=0, TargetMinimum=1, bClamp=true, Value=@"Get Delta Ref_1".Result.Location.X).Result
   )
   Get Delta Ref_2 = RigVMFunctionReferenceNode(Item=(Type=Bone,Name="pen_eye_flipbook"), ExecuteContext=@RigVMFunction_Sequence.C)
   > &RigUnit_SetCurveValue(
    Curve=Eye Sad
    Value=&RigVMFunction_MathFloatRemap(SourceMaximum=2.9, SourceMinimum=2, TargetMaximum=0, TargetMinimum=1, bClamp=true, Value=@"Get Delta Ref_2".Result.Location.X).Result
   )
  ]}
]
hierarchy [
 root {loc=<0 0 0>, rot=<0 0 0 1>}
 pack {parent=root, loc=<0 -0.7316 0>, rot=<0 0 0 1>}
 pen_pelvis {parent=pack, loc=<0 -10.6536 -10.0427>, rot=<0.7071 0 0.7071 0>}
 pen_spine_01 {parent=pen_pelvis, loc=<1.4403 0 0>, rot=<0 0 0 1>}
 pen_spine_02 {parent=pen_spine_01, loc=<3.8891 0 0>, rot=<0 0 0 1>}
 pen_neck {parent=pen_spine_02, loc=<5.6503 0 0>, rot=<0 0 0 1>}
 pen_head {parent=pen_neck, loc=<2.1279 0 0>, rot=<0 0 0 1>}
 pen_eye_flipbook {parent=pen_head, loc=<0 0 0>, rot=<0 0 0 1>}
 pen_base_jaw {parent=pen_head, loc=<1.1761 6.4818 0>, rot=<0 0 -0.7071 -0.7071>}
 pen_upper_jaw {parent=pen_base_jaw, loc=<0.4068 -0.0002 0>, rot=<0 0 0 1>}
 pen_lower_jaw {parent=pen_base_jaw, loc=<0.4068 -0.0002 0>, rot=<0 0 0 1>}
 pen_tongue_01 {parent=pen_lower_jaw, loc=<0 0.3662 0>, rot=<0 0 0 1>}
 pen_tongue_02 {parent=pen_tongue_01, loc=<0.8611 0.0446 0>, rot=<0 0 0 1>}
 pen_clavicle_l {parent=pen_spine_02, loc=<4.3578 0.5814 -5.5144>, rot=<0.0401 -0.765 0.0336 -0.6419>}
 pen_upperarm_l {parent=pen_clavicle_l, loc=<3.7911 0 0>, rot=<-0.0033 -0.1277 -0.026 -0.9915>}
 pen_lowerarm_l {parent=pen_upperarm_l, loc=<3.5922 0 0>, rot=<0 0 -0.0262 -0.9997>}
 pen_hand_l {parent=pen_lowerarm_l, loc=<2.9181 0 0>, rot=<0 0 0 1>}
 pen_clavicle_r {parent=pen_spine_02, loc=<4.3578 0.5814 5.5147>, rot=<0.765 0.0401 -0.6419 -0.0336>}
 pen_upperarm_r {parent=pen_clavicle_r, loc=<-3.7916 0 0.0016>, rot=<-0.0033 -0.1277 -0.026 -0.9915>}
 pen_lowerarm_r {parent=pen_upperarm_r, loc=<-3.5922 0 0>, rot=<0 0 -0.0262 -0.9997>}
 pen_hand_r {parent=pen_lowerarm_r, loc=<-2.9181 0 0>, rot=<0 0 0 1>}
 pen_foot_r {parent=pen_pelvis, loc=<-2.2523 -0.2026 3.9451>, rot=<0 -0.1305 0.9914 0>}
 pen_ball_r {parent=pen_foot_r, loc=<0.4 -1.4796 0>, rot=<0 0 0.7071 -0.7071>}
 pen_foot_l {parent=pen_pelvis, loc=<-2.2523 -0.2025 -3.9446>, rot=<0.1305 0 0 -0.9914>}
 pen_ball_l {parent=pen_foot_l, loc=<-0.4002 1.4796 0>, rot=<0 0 0.7071 -0.7071>}
 Eye Sad {type=Curve, loc=<0 0 0>, rot=<0 0 0 1>}
 Eye Open {type=Curve, loc=<0 0 0>, rot=<0 0 0 1>}
 Eye Close {type=Curve, loc=<0 0 0>, rot=<0 0 0 1>}
]
