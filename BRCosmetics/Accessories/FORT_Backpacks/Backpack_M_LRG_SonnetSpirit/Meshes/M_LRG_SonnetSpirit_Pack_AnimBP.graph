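# Animation Blueprint for the SonnetSpirit backpack cosmetic, animating idles and reactions to gameplay events.
# A front-end state machine replays a reaction on delay timers, while the in-game one plays one-shots from brief flags set by elimination, chest-open and damage input events.
# The chosen pose runs through a Control Rig, then two additive pack-bone offsets driven by owner spine bend and melee arm pose.
# Updates also cache eye curve values.
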AnimBlueprint /BRCosmetics/Accessories/FORT_Backpacks/Backpack_M_LRG_SonnetSpirit/Meshes/M_LRG_SonnetSpirit_Pack_AnimBP
target_skeleton=&/M_LRG_SonnetSpirit_Pack_Skeleton
graphs [
 AnimGraph {nodes [
   AnimGraphNode_StateMachine_1 = FE_Reactivity() {is=title AnimGraphNode_StateMachine}
   AnimGraphNode_StateMachine_2 = Idle and Reactive() {is=title AnimGraphNode_StateMachine}
   n1 = Control Rig(Source=Blend Poses by bool(BlendPose_0=@AnimGraphNode_StateMachine_1, BlendPose_1=@AnimGraphNode_StateMachine_2, bActiveValue=Get bIsInFrontEnd()) {blendType=Linear}) {controlRigAssetReference {blueprintRigClass=&/M_LRG_SonnetSpirit_Pack_CtrlRig.M_LRG_SonnetSpirit_Pack_CtrlRig_C}}
   n2 = "Transform (Modify) Bone - Bone: pack"(
    ComponentPose=Local To Component(LocalPose=@n1)
    Translation=Get Backpack Translation() -> (Backpack Translation:Vector)
    Rotation=Get Backpack Rotate() -> (Backpack Rotate:Rotator)
   ) {
    rotationMode=BMM_Additive
    rotationSpace=BCS_BoneSpace
    translationMode=BMM_Additive
    translationSpace=BCS_BoneSpace
    alphaBoolBlend {blendOption=Linear}
    boneToModify {boneName=pack}
   }
   n3 = "Transform (Modify) Bone - Bone: pack"(
    ComponentPose=@n2
    Alpha=Get isMelee() -> (isMelee:double)
    Translation="0,-1.5,0"
    Rotation=Get Backpack Rotate Pickaxe() -> (Backpack Rotate Pickaxe:Rotator)
   ) {
    rotationMode=BMM_Additive
    rotationSpace=BCS_BoneSpace
    translationMode=BMM_Additive
    translationSpace=BCS_BoneSpace
    alphaBoolBlend {blendOption=Linear}
    alphaScaleBiasClamp {interpSpeedDecreasing=5, interpSpeedIncreasing=5}
    boneToModify {boneName=pack}
    rotation {roll=20}
    translation {y=-1.5}
   }
   Output Pose(Result=Component To Local(ComponentPose=@n3)) {layerGroup=None}
  ]}
 FE_Reactivity {
  entry=Idle
  path=AnimGraph.AnimGraphNode_StateMachine_1.FE_Reactivity
  ids=id
  nodes [
   AnimStateNode_0 = Idle()
   AnimStateNode_2 = FE_Reactive() {is=title AnimStateNode, stateEntered {guid=#1, notifyName=Entered FE_Reactive_Anim}}
   AnimStateTransitionNode_8 = AnimStateTransitionNode()
   AnimStateTransitionNode_9 = AnimStateTransitionNode() {bAutomaticRuleBasedOnSequencePlayerInState=true}
  ]
  links [
   {
    dst=Idle
    rule=AnimGraph.AnimGraphNode_StateMachine_1.FE_Reactivity.AnimStateTransitionNode_9.Transition
    src=FE_Reactive
    via=AnimStateTransitionNode_9
   }
   {
    dst=FE_Reactive
    rule=AnimGraph.AnimGraphNode_StateMachine_1.FE_Reactivity.AnimStateTransitionNode_8.Transition
    src=Idle
    via=AnimStateTransitionNode_8
   }
  ]
 }
 Idle {
  path=AnimGraph.AnimGraphNode_StateMachine_1.FE_Reactivity.AnimStateNode_0.Idle
  nodes [
   n1 = Random Sequence Player!() -> (Pose:Pose Link Structure) {
    is=title AnimGraphNode_RandomPlayer
    entries [
     {
      chanceToPlay=0.8
      maxLoopCount=0
      maxPlayRate=1
      minLoopCount=0
      minPlayRate=1
      sequence=/BRCosmetics/Accessories/FORT_Backpacks/Backpack_M_LRG_SonnetSpirit/Animation/SonnetSpirit_Idle_PACK
      blendIn {blendOption=Linear, blendTime=0.2, customCurve=null}
     }
     {
      chanceToPlay=0.2
      maxLoopCount=0
      maxPlayRate=1
      minLoopCount=0
      minPlayRate=1
      sequence=/BRCosmetics/Accessories/FORT_Backpacks/Backpack_M_LRG_SonnetSpirit/Animation/SonnetSpirit_IdleBreak_PACK
      blendIn {blendOption=Linear, blendTime=0.2, customCurve=null}
     }
     {
      chanceToPlay=0.2
      maxLoopCount=0
      maxPlayRate=1
      minLoopCount=0
      minPlayRate=1
      sequence=/BRCosmetics/Accessories/FORT_Backpacks/Backpack_M_LRG_SonnetSpirit/Animation/SonnetSpirit_IdleBreak2_PACK
      blendIn {blendOption=Linear, blendTime=0.2, customCurve=null}
     }
    ]
   }
   Output Animation Pose(Result=@n1)
  ]
 }
 FE_Reactive {
  path=AnimGraph.AnimGraphNode_StateMachine_1.FE_Reactivity.AnimStateNode_2.FE_Reactive
  nodes [
   n1 = Sequence Player 'SonnetSpirit_Elimination_PACK'() {bLoopAnimation=false, sequence=/BRCosmetics/Accessories/FORT_Backpacks/Backpack_M_LRG_SonnetSpirit/Animation/SonnetSpirit_Elimination_PACK}
   Output Animation Pose(Result=@n1)
  ]
 }
 Transition {
  path=AnimGraph.AnimGraphNode_StateMachine_1.FE_Reactivity.AnimStateTransitionNode_8.Transition
  nodes [Result(bCanEnterTransition=Get Play Reactive FE Anim() -> (Play Reactive FE Anim:bool))]
 }
 Transition {path=AnimGraph.AnimGraphNode_StateMachine_1.FE_Reactivity.AnimStateTransitionNode_9.Transition, nodes [Result()]}
 Idle and Reactive {
  entry=Idle
  path=AnimGraph.AnimGraphNode_StateMachine_2.Idle and Reactive
  ids=id
  nodes [
   AnimStateNode_0 = Reactive_OpenChest() {
    is=title AnimStateNode
    stateEntered {guid=#2}
    stateFullyBlended {guid=#3}
    stateLeft {guid=#4}
   }
   AnimStateNode_1 = Reactive_Elimination() {
    is=title AnimStateNode
    stateEntered {guid=#2}
    stateFullyBlended {guid=#3}
    stateLeft {guid=#4}
   }
   AnimStateNode_2 = Idle() {
    stateEntered {guid=#5}
    stateFullyBlended {guid=#6}
    stateLeft {guid=#7}
   }
   AnimStateNode_3 = Reactive_TakeDamage() {
    is=title AnimStateNode
    stateEntered {guid=#8}
    stateFullyBlended {guid=#9}
    stateLeft {guid=#10}
   }
   AnimStateTransitionNode_0 = AnimStateTransitionNode() {bAllowInertializationForSelfTransitions=true}
   AnimStateTransitionNode_1 = AnimStateTransitionNode() {bAllowInertializationForSelfTransitions=true, bAutomaticRuleBasedOnSequencePlayerInState=true}
   AnimStateTransitionNode_2 = AnimStateTransitionNode() {
    crossfadeDuration=0.3
    transitionEnd {guid=#11}
    transitionInterrupt {guid=#12}
    transitionStart {guid=#13}
   }
   AnimStateTransitionNode_3 = AnimStateTransitionNode() {
    bAutomaticRuleBasedOnSequencePlayerInState=true
    crossfadeDuration=0.1
    transitionEnd {guid=#14}
    transitionInterrupt {guid=#15}
    transitionStart {guid=#16}
   }
   AnimStateTransitionNode_4 = AnimStateTransitionNode() {bAllowInertializationForSelfTransitions=true}
   AnimStateTransitionNode_5 = AnimStateTransitionNode() {bAllowInertializationForSelfTransitions=true, bAutomaticRuleBasedOnSequencePlayerInState=true}
  ]
  links [
   {
    dst=Reactive_Elimination
    rule=AnimGraph.AnimGraphNode_StateMachine_2.Idle and Reactive.AnimStateTransitionNode_2.Transition
    src=Idle
    via=AnimStateTransitionNode_2
   }
   {
    dst=Reactive_OpenChest
    rule=AnimGraph.AnimGraphNode_StateMachine_2.Idle and Reactive.AnimStateTransitionNode_4.Transition
    src=Idle
    via=AnimStateTransitionNode_4
   }
   {
    dst=Reactive_TakeDamage
    rule=AnimGraph.AnimGraphNode_StateMachine_2.Idle and Reactive.AnimStateTransitionNode_0.Transition
    src=Idle
    via=AnimStateTransitionNode_0
   }
   {
    dst=Idle
    rule=AnimGraph.AnimGraphNode_StateMachine_2.Idle and Reactive.AnimStateTransitionNode_3.Transition
    src=Reactive_Elimination
    via=AnimStateTransitionNode_3
   }
   {
    dst=Idle
    rule=AnimGraph.AnimGraphNode_StateMachine_2.Idle and Reactive.AnimStateTransitionNode_5.Transition
    src=Reactive_OpenChest
    via=AnimStateTransitionNode_5
   }
   {
    dst=Idle
    rule=AnimGraph.AnimGraphNode_StateMachine_2.Idle and Reactive.AnimStateTransitionNode_1.Transition
    src=Reactive_TakeDamage
    via=AnimStateTransitionNode_1
   }
  ]
 }
 Reactive_OpenChest {
  path=AnimGraph.AnimGraphNode_StateMachine_2.Idle and Reactive.AnimStateNode_0.Reactive_OpenChest
  nodes [
   n1 = Sequence Player 'SonnetSpirit_Elimination_PACK'() {bLoopAnimation=false, sequence=&SonnetSpirit_Elimination_PACK}
   Output Animation Pose(Result=@n1)
  ]
 }
 Reactive_Elimination {
  path=AnimGraph.AnimGraphNode_StateMachine_2.Idle and Reactive.AnimStateNode_1.Reactive_Elimination
  nodes [
   n1 = Sequence Player 'SonnetSpirit_Elimination_PACK'() {bLoopAnimation=false, sequence=&SonnetSpirit_Elimination_PACK}
   Output Animation Pose(Result=@n1)
  ]
 }
 Idle {
  path=AnimGraph.AnimGraphNode_StateMachine_2.Idle and Reactive.AnimStateNode_2.Idle
  nodes [
   n1 = Random Sequence Player() {entries [
     {
      chanceToPlay=0.6
      maxLoopCount=0
      maxPlayRate=1
      minLoopCount=0
      minPlayRate=1
      sequence=&SonnetSpirit_Idle_PACK
      blendIn {blendOption=Linear, blendTime=0.2, customCurve=null}
     }
     {
      chanceToPlay=0.3
      maxLoopCount=0
      maxPlayRate=1
      minLoopCount=0
      minPlayRate=1
      sequence=&SonnetSpirit_IdleBreak_PACK
      blendIn {blendOption=Linear, blendTime=0.2, customCurve=null}
     }
     {
      chanceToPlay=0.3
      maxLoopCount=0
      maxPlayRate=1
      minLoopCount=0
      minPlayRate=1
      sequence=&SonnetSpirit_IdleBreak2_PACK
      blendIn {blendOption=Linear, blendTime=0.2, customCurve=null}
     }
    ]}
   Output Animation Pose(Result=@n1)
  ]
 }
 Reactive_TakeDamage {
  path=AnimGraph.AnimGraphNode_StateMachine_2.Idle and Reactive.AnimStateNode_3.Reactive_TakeDamage
  nodes [
   n1 = Sequence Player 'SonnetSpirit_Damage_PACK'() {bLoopAnimation=false, sequence=/BRCosmetics/Accessories/FORT_Backpacks/Backpack_M_LRG_SonnetSpirit/Animation/SonnetSpirit_Damage_PACK}
   Output Animation Pose(Result=@n1)
  ]
 }
 Transition {
  path=AnimGraph.AnimGraphNode_StateMachine_2.Idle and Reactive.AnimStateTransitionNode_0.Transition
  nodes [Result(bCanEnterTransition=Get Reactive TakeDamage() -> (Reactive TakeDamage:bool))]
 }
 Transition {path=AnimGraph.AnimGraphNode_StateMachine_2.Idle and Reactive.AnimStateTransitionNode_1.Transition, nodes [Result()]}
 Transition {
  path=AnimGraph.AnimGraphNode_StateMachine_2.Idle and Reactive.AnimStateTransitionNode_2.Transition
  nodes [Result(bCanEnterTransition=Get Reactive Elimination() -> (Reactive Elimination:bool))]
 }
 Transition {path=AnimGraph.AnimGraphNode_StateMachine_2.Idle and Reactive.AnimStateTransitionNode_3.Transition, nodes [Result()]}
 Transition {
  path=AnimGraph.AnimGraphNode_StateMachine_2.Idle and Reactive.AnimStateTransitionNode_4.Transition
  nodes [Result(bCanEnterTransition=Get Reactive OpenChest() -> (Reactive OpenChest:bool))]
 }
 Transition {path=AnimGraph.AnimGraphNode_StateMachine_2.Idle and Reactive.AnimStateTransitionNode_5.Transition, nodes [Result()]}
 EventGraph {nodes [
   Event OnCharacterCustomizationLoaded()
   > n1 = Cast To FortPlayerPawn(Object=TryGetFortPawn())
   > n2 = Branch()
   > Do Once(Reset=@n4.then)
   > n3 = Delay(Duration=3)
   Set Play Reactive FE Anim(execute=[@n5.then, @n3.then], Play Reactive FE Anim:bool=true) -> (Output_Get:bool)
   n4 = Set Play Reactive FE Anim(execute=@n2.else, Play Reactive FE Anim=false)
   AnimNotify_FE_ReactiveAnim() {is=title K2Node_Event}
   > Set Play Reactive FE Anim(Play Reactive FE Anim=false)
   > n5 = Delay(Duration=25)
   n6 = Event BlueprintProcessAnimInputEvent()
   > n7 = Sequence!(execute) -> (then_0, then_1, then_2:Exec)
   > Branch(Condition=IsNamed(self=@n6.AnimInputEvent, InEventName=aie.EnemyPlayerEliminated))
   > Set Reactive Elimination(Reactive Elimination:bool=true) -> (Output_Get:bool)
   > RetriggerableDelay(Duration=0.2)
   > Set Reactive Elimination(Reactive Elimination=false)
   Branch(execute=@n7.then_1, Condition=IsNamed(self=@n6.AnimInputEvent, InEventName=aie.ChestOpened))
   > Set Reactive OpenChest(Reactive OpenChest:bool=true) -> (Output_Get:bool)
   > RetriggerableDelay(Duration=0.2)
   > Set Reactive OpenChest(Reactive OpenChest=false)
   Branch(
    execute=@n7.then_2
    Condition=OR Boolean(
     A=IsNamed(self=@n6.AnimInputEvent, InEventName=aie.PlayerVitality.ShieldsDecreased)
     B=IsNamed(self=@n6.AnimInputEvent, InEventName=aie.PlayerVitality.HealthDecreased)
    )
   )
   > Set Reactive TakeDamage(Reactive TakeDamage:bool=true) -> (Output_Get:bool)
   > RetriggerableDelay(Duration=0.2)
   > Set Reactive TakeDamage(Reactive TakeDamage=false)
   Event BlueprintPostEvaluateAnimation() {is=title K2Node_Event}
   > Set Reactive Elimination(Reactive Elimination=false)
   Event BlueprintUpdateAnimation()
   > n8 = Sequence()
   > Set EyeOpen(EyeOpen:float=GetCurveValue(CurveName=Eye Open)) -> (Output_Get:float)
   > Set EyeClose(EyeClose:float=GetCurveValue(CurveName=Eye Close)) -> (Output_Get:float)
   > Set EyeSad(EyeSad:float=GetCurveValue(CurveName=Eye Sad)) -> (Output_Get:float)
   n9 = GetOwnerSkeletalMeshComponent()
   n10 = Set head down rot(
    execute=@n8.then_1
    head down rot:double=GetAlphaFromReferenceTransformDelta(SkeletalMeshComponent=@n9, BoneName=spine_01, BaseName=head, InputRange=(X=0,Y=-45), TwistAxis="1,0,0")
   ) -> (Output_Get:double)
   n11 = "To Float (Boolean)"(InBool=Get bIsPlayingMeleeAnim())
   n12 = float * float(B=0.2) {is=title K2Node_PromotableOperator}
   n13 = Get isMelee()
   > Set isMelee(isMelee:double=float * float(A=FInterpTo(Current=@n13, Target=@n11, DeltaTime=@n12, InterpSpeed=5), B=@n10)) -> (Output_Get:double)
   n14 = BreakRotator(InRot=BreakTransform(InTransform=GetDeltaTransformFromRefPose(self=@n9, BoneName=upperarm_l, BaseName=clavicle_l)).Rotation)
   > Set Backpack Rotate Pickaxe(Backpack Rotate Pickaxe:Rotator=MakeRotator(Roll=MapRangeClamped(Value=@n14.Yaw, InRangeA=0, InRangeB=90, OutRangeB=30))) -> (Output_Get:Rotator)
   n15 = GetOwnerSkeletalMeshComponent()
   n16 = BreakRotator(InRot=BreakTransform(InTransform=GetDeltaTransformFromRefPose(self=@n15, BoneName=spine_05, BaseName=spine_01)).Rotation)
   n17 = MapRangeClamped(Value=@n16.Yaw, InRangeA=-10, InRangeB=-50, OutRangeB=50)
   Set Backpack Rotate(execute=@n8.then_2, Backpack Rotate:Rotator=MakeRotator(Roll=@n17, Pitch=float / float(A=@n17, B=5))) -> (Output_Get:Rotator)
   > Set Backpack Translation(Backpack Translation:Vector=MakeVector(Y=MapRangeClamped(Value=@n16.Yaw, InRangeA=-15, InRangeB=-50, OutRangeB=-10))) -> (Output_Get:Vector)
   TryGetPawnOwner()
   GetModifierForPartType!(self:Fort Player Pawn Object Reference=@n1, PartType:EFortCustomPartType Enum=Backpack) -> (ReturnValue:Custom Character Part Modifier Object Reference)
  ]}
]
variables [
 Backpack Rotate {type=Rotator}
 Backpack Rotate Pickaxe {type=Rotator}
 Backpack Translation {type=Vector}
 EyeClose {type=number}
 EyeOpen {type=number}
 EyeSad {type=number}
 head down rot {type=number}
 isMelee {type=number}
 Play Reactive FE Anim {type=boolean}
 Play Reactive Front End Anim {type=boolean}
 Reactive Elimination {type=boolean}
 Reactive OpenChest {type=boolean}
 Reactive TakeDamage {type=boolean}
 Reactivity Enabled {type=boolean}
]
